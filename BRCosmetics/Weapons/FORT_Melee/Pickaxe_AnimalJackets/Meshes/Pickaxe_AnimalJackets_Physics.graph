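# Physics asset for the AnimalJackets two-handed pickaxe, covering its handle and two exhaust door bones named as dynamic.
# A long capsule covers the handle, and a short rotated capsule sits on each of the left and right doors.
# Each door is constrained to the handle with both swing axes locked, so it can rotate only about the twist axis, like a hinged flap.
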PhysicsAsset /BRCosmetics/Weapons/FORT_Melee/Pickaxe_AnimalJackets/Meshes/Pickaxe_AnimalJackets_Physics
bodies {
 Melee_TwoHanded_Handle [{shapeType=Capsule, length=80, radius=3, center=<0 0 22.7279>}]
 dyn_exhaustdoor_l [{shapeType=Capsule, length=6.4795, radius=4.1232, center=<0 -3.6625 0>, rotation=<rot 0 0 90>}]
 dyn_exhaustdoor_r [{shapeType=Capsule, length=6.4795, radius=4.1232, center=<0 -3.6625 0>, rotation=<rot 0 0 90>}]
}
constraints [
 {
  bone1Name=dyn_exhaustdoor_l
  bone2Name=Melee_TwoHanded_Handle
  swing1LimitDegrees=45
  swing1Motion=Locked
  swing2LimitDegrees=45
  swing2Motion=Locked
  twistLimitDegrees=20
 }
 {
  bone1Name=dyn_exhaustdoor_r
  bone2Name=Melee_TwoHanded_Handle
  swing1LimitDegrees=45
  swing1Motion=Locked
  swing2LimitDegrees=45
  swing2Motion=Locked
  twistLimitDegrees=20
 }
]
