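# Animation Blueprint for the AnimalJackets pickaxe cosmetic skeleton, adding secondary physics motion to the weapon's incoming pose.
# The input pose runs through a half-weight RigidBody node using the pickaxe physics asset, simulated relative to b_AR_Root with overridden gravity, then goes to output.
# A chained pair of Anim Dynamics nodes for the left and right exhaust-door bones is present but unconnected to the output.
# The event graph holds no logic.
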AnimBlueprint /BRCosmetics/Weapons/FORT_Melee/Pickaxe_AnimalJackets/Meshes/Pickaxe_AnimalJackets_Skeleton_AnimBlueprint
target_skeleton=&/Pickaxe_AnimalJackets_Skeleton
graphs [
 AnimGraph {nodes [
   n1 = Anim Dynamics - Bone: dyn_exhaustdoor_l(Alpha=1, ExternalForce="0,0,0") {
    angularDampingOverride=0.9999
    bGravityOverrideInSimSpace=true
    simulationSpace=RootRelative
    alphaBoolBlend {blendOption=Linear}
    boundBone {boneName=dyn_exhaustdoor_l}
    componentLinearAccScale=<1 1 1>
    componentLinearVelScale=<1 1 1>
    physicsBodyDefinitions [{
      sphereCollisionRadius=0
      boundBone {boneName=dyn_exhaustdoor_l}
      boxExtents=<10 3 10>
      constraintSetup {angularTargetAxis=AxisY, twistAxis=AxisY, angularLimitsMax=<5 0 0>, angularLimitsMin=<-30 0 0>, angularTarget=<0 0 0>}
      localJointOffset=<0 3 0>
     }]
   }
   Anim Dynamics - Bone: dyn_exhaustdoor_r(ComponentPose=@n1, Alpha=1, ExternalForce="0,0,0") {
    angularDampingOverride=0.9999
    bGravityOverrideInSimSpace=true
    simulationSpace=RootRelative
    alphaBoolBlend {blendOption=Linear}
    boundBone {boneName=dyn_exhaustdoor_r}
    componentLinearAccScale=<1 1 1>
    componentLinearVelScale=<1 1 1>
    physicsBodyDefinitions [{
      sphereCollisionRadius=0
      boundBone {boneName=dyn_exhaustdoor_r}
      boxExtents=<10 3 10>
      constraintSetup {angularTargetAxis=AxisY, twistAxis=AxisY, angularLimitsMax=<30 0 0>, angularLimitsMin=<-5 0 0>, angularTarget=<0 0 0>}
      localJointOffset=<0 3 0>
     }]
   }
   AnimGraphNode_SubInput_0 = Input Pose()
   n2 = RigidBody(ComponentPose=Local To Component(LocalPose=@AnimGraphNode_SubInput_0), Alpha=0.5) {
    bOverrideWorldGravity=true
    overridePhysicsAsset=&/Pickaxe_AnimalJackets_Physics
    simulationSpace=BaseBoneSpace
    alphaBoolBlend {blendOption=Linear}
    baseBoneRef {boneName=b_AR_Root}
    componentLinearAccScale=<1 1 1>
    componentLinearVelScale=<1 1 1>
    overrideWorldGravity {z=-980}
   }
   Output Pose(Result=Component To Local(ComponentPose=@n2)) {layerGroup=None}
  ]}
 EventGraph {nodes [Event BlueprintUpdateAnimation()]}
]
variables []
